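# Sound cue for AI Gallery NPC voice lines in the 2024 "OnCommand" set, likely barks played when the AI responds to a command.
# A Random node picks one of six equally weighted wave players and is excluded from branch culling.
# It plays louder than default, using the shared NPC alerted-VO attenuation, the non-critical enemy VO sound class and an AI Gallery chatter concurrency limit.
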SoundCue /AIGallery/Audio/VO/AIGallery/VO_AIGallery_2024_OnCommand_Cue
root=Random_0
nodes [
 {
  id=Random_0
  children [WavePlayer_0,WavePlayer_1,WavePlayer_2,WavePlayer_3,WavePlayer_4,WavePlayer_5]
  properties {bShouldExcludeFromBranchCulling=true, weights [1,1,1,1,1,1]}
 }
 {id=WavePlayer_0, properties {soundWaveAssetPtr=/AIGallery/Audio/VO/AIGallery/Waves/VO_AIGallery_2024_OnCommand_010}}
 {id=WavePlayer_1, properties {soundWaveAssetPtr=/AIGallery/Audio/VO/AIGallery/Waves/VO_AIGallery_2024_OnCommand_020}}
 {id=WavePlayer_2, properties {soundWaveAssetPtr=/AIGallery/Audio/VO/AIGallery/Waves/VO_AIGallery_2024_OnCommand_030}}
 {id=WavePlayer_3, properties {soundWaveAssetPtr=/AIGallery/Audio/VO/AIGallery/Waves/VO_AIGallery_2024_OnCommand_040}}
 {id=WavePlayer_4, properties {soundWaveAssetPtr=/AIGallery/Audio/VO/AIGallery/Waves/VO_AIGallery_2024_OnCommand_050}}
 {id=WavePlayer_5, properties {soundWaveAssetPtr=/AIGallery/Audio/VO/AIGallery/Waves/VO_AIGallery_2024_OnCommand_060}}
]
properties {
 attenuationSettings=/Game/Sounds/Fort_Human_VO_BR/NPCs/Shared/NPC_VO_Alerted_Att
 maxDistance=12500
 soundClassObject=/Game/Sounds/FortSoundClassesAndModes/Fort_VO_Enemy_NonCritical
 volumeMultiplier=2
 concurrencySet [&/AIGallery_Chatter_Conc]
}
