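# Sound cue for the AI Gallery "Rescued End" voice line, playing Chewbacca vocalizations.
# A Random node picks one of six equally weighted wave players.
# It uses the shared alerted-NPC attenuation, the critical enemy VO sound class and the AI Gallery chatter concurrency set, with boosted volume.
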SoundCue /AIGallery/Audio/VO/AIGallery/VO_AIGallery_2024_Rescued_End_Cue
root=Random_0
nodes [
 {
  id=Random_0
  children [WavePlayer_0, WavePlayer_1, WavePlayer_2, WavePlayer_3, WavePlayer_10, WavePlayer_11]
  properties {weights [1,1,1,1,1,1]}
 }
 {id=WavePlayer_0, properties {soundWaveAssetPtr=/AIGallery/Audio/VO/AIGallery/Waves/VO_Chewbacca_2024_Rescued_End_010}}
 {id=WavePlayer_1, properties {soundWaveAssetPtr=/AIGallery/Audio/VO/AIGallery/Waves/VO_Chewbacca_2024_Rescued_End_020}}
 {id=WavePlayer_10, properties {soundWaveAssetPtr=/AIGallery/Audio/VO/AIGallery/Waves/VO_Chewbacca_2024_Rescued_End_050}}
 {id=WavePlayer_11, properties {soundWaveAssetPtr=/AIGallery/Audio/VO/AIGallery/Waves/VO_Chewbacca_2024_Rescued_End_060}}
 {id=WavePlayer_2, properties {soundWaveAssetPtr=/AIGallery/Audio/VO/AIGallery/Waves/VO_Chewbacca_2024_Rescued_End_030}}
 {id=WavePlayer_3, properties {soundWaveAssetPtr=/AIGallery/Audio/VO/AIGallery/Waves/VO_Chewbacca_2024_Rescued_End_040}}
]
properties {
 attenuationSettings=/Game/Sounds/Fort_Human_VO_BR/NPCs/Shared/NPC_VO_Alerted_Att
 maxDistance=12500
 soundClassObject=&Fort_VO_Enemy_Critical
 volumeMultiplier=2
 concurrencySet [&/AIGallery_Chatter_Conc]
}
